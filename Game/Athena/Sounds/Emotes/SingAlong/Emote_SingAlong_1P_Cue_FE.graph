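# Sound cue for the SingAlong emote's first-person music, built from four looping stem layers.
# A crossfade on the GroupNum parameter keeps layer 1 as the base and stacks layers 2 to 4 as the value rises.
# Envelopers ramp the upper three layers in over time, with layer 3 slightly delayed.
# It uses emote music class and concurrency settings and sends to a music audio-analysis submix.
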SoundCue /Game/Athena/Sounds/Emotes/SingAlong/Emote_SingAlong_1P_Cue_FE
root=ParamCrossFade_0
nodes [
 {
  id=Enveloper_0
  children [WavePlayer_8]
  properties {pitchCurve {editorCurveData {keys [{time=0}]}}, volumeCurve {editorCurveData {keys [<at 0 0>,<at 1 1>]}}}
 }
 {
  id=Enveloper_1
  children [WavePlayer_2]
  properties {pitchCurve {editorCurveData {keys [{time=0}]}}, volumeCurve {editorCurveData {keys [<at 0.4015 0>,<at 1.4387 1>]}}}
 }
 {
  id=Enveloper_4
  children [WavePlayer_3]
  properties {pitchCurve {editorCurveData {keys [{time=0}]}}, volumeCurve {editorCurveData {keys [<at 0 0>,<at 1 1>]}}}
 }
 {
  id=ParamCrossFade_0
  children [WavePlayer_1,Enveloper_4,Enveloper_1,Enveloper_0]
  properties {
   paramName=GroupNum
   crossFadeInput [
    {fadeInDistanceEnd=0, fadeInDistanceStart=0, fadeOutDistanceEnd=100, fadeOutDistanceStart=100, volume=1}
    {fadeInDistanceEnd=2, fadeInDistanceStart=1, fadeOutDistanceEnd=100, fadeOutDistanceStart=100, volume=1}
    {fadeInDistanceEnd=3, fadeInDistanceStart=2, fadeOutDistanceEnd=100, fadeOutDistanceStart=100, volume=1}
    {fadeInDistanceEnd=4, fadeInDistanceStart=3, fadeOutDistanceEnd=100, fadeOutDistanceStart=100, volume=1}
   ]
  }
 }
 {id=WavePlayer_1, properties {bLooping=true, soundWaveAssetPtr=&/Emote_SingAlong_Layer_01}}
 {id=WavePlayer_2, properties {bLooping=true, soundWaveAssetPtr=&/Emote_SingAlong_Layer_03}}
 {id=WavePlayer_3, properties {bLooping=true, soundWaveAssetPtr=&/Emote_SingAlong_Layer_02}}
 {id=WavePlayer_8, properties {bLooping=true, soundWaveAssetPtr=&/Emote_SingAlong_Layer_04}}
]
properties {
 maxDistance=2097152
 soundClassObject=/Game/Sounds/FortSoundClassesAndModes/Fort_Music_Emote_1P_PSM
 volumeMultiplier=1
 concurrencyOverrides {maxCount=3, resolutionRule=StopOldest}
 concurrencySet [/Game/Sounds/Concurrency_Rules/Athena_EmoteMusic_1P_Conc]
 soundSubmixSends [{minSendLevel=0, soundSubmix=&Fortnite_AudioAnalysis_Music}]
}
